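# Material for golf course tee sign numbers, picking the displayed digit from a texture atlas through a scalar Number parameter.
# Number minus one becomes the frame index for two CustomNumbers_SubUV calls that share UVs and a constant sub-image count.
# One call samples the _D texture, while the other samples the _S mask texture and splits out its red and blue channels.
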Material /Figment_S10/Environments/Props/Golf_Course/Materials/M_GolfCourse_Sign_Numbers
graph {
 properties {pixelDepthOffsetMode=PDOM_Legacy, usedShadingModels=Default Lit}
 nodes [
  n1 = TextureCoordinate()
  n2 = Constant() {r=3}
  n3 = Add(A=ScalarParameter() {defaultValue=7, parameterName=Number}) {constB=-1}
  n4 = MaterialFunctionCall!(
   Texture=TextureObject() {samplerType=SAMPLERTYPE_Masks, texture=/Figment_S10/Environments/Props/Golf_Course/Textures/T_GolfCourse_Tee_Sign_Numbers_S}
   UVs=@n1
   SubImages=@n2
   Frame=@n3
  ) {materialFunction=/Game/Weapons/WeaponSkins/MaterialFunction/CustomNumbers_SubUV}
  ComponentMask(None=@n4.RGB) {r=true}
  ComponentMask(None=@n4.RGB) {b=true}
  MaterialFunctionCall(
   Texture=TextureObject() {texture=/Figment_S10/Environments/Props/Golf_Course/Textures/T_GolfCourse_Tee_Sign_Numbers_D}
   UVs=@n1
   SubImages=@n2
   Frame=@n3
  ) {materialFunction=&CustomNumbers_SubUV}
 ]
}
